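# Physics asset for the swim trunks on the Alien Summer medium male player body cosmetic, covering secondary motion of trunks and drawstring.
# Box bodies form four front and back leg panels, and two chained capsules form the string.
# Each panel and the string root hangs off a tiny MSTR anchor capsule with tight angular limits.
# Anchors and thigh capsules connect to the pelvis capsule with free rotation.
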PhysicsAsset /BRCosmetics/Characters/Player/Male/Medium/Bodies/M_MED_Alien_Summer/Meshes/M_MED_Alien_Summer_Trunks_Physics
bodies {
 dyn_trunks_lt_bk [{shapeType=Box, extentX=11.5, extentY=3.5233, extentZ=16, shapeName=dyn_trunks_lt_bk_box, center=<9 0 0>, rotation=<rot 0 0 92.0772>}]
 dyn_trunks_lt_bk_MSTR [{shapeType=Capsule, length=0.505, radius=0.505, rotation=<rot 0 -90 0>}]
 dyn_trunks_lt_fr [{
   shapeType=Box
   extentX=11.5
   extentY=2.6113
   extentZ=16
   shapeName=dyn_trunks_lt_fr_box
   center=<10.0683 0 0>
   rotation=<rot -6.3169 0 87.7369>
  }]
 dyn_trunks_lt_fr_MSTR [{shapeType=Capsule, length=0.505, radius=0.505, rotation=<rot 0 -90 0>}]
 dyn_trunks_rt_bk [{shapeType=Box, extentX=11.5, extentY=3.5233, extentZ=16, shapeName=dyn_trunks_rt_bk_box, center=<9 0 0>, rotation=<rot -0.2164 0 89.9999>}]
 dyn_trunks_rt_bk_MSTR [{shapeType=Capsule, length=0.505, radius=0.505, rotation=<rot 0 -90 0>}]
 dyn_trunks_rt_fr [{shapeType=Box, extentX=11.5, extentY=2.9109, extentZ=16, shapeName=dyn_trunks_rt_fr_box, center=<10.0683 0 0>, rotation=<rot -10.1003 0 90>}]
 dyn_trunks_rt_fr_MSTR [{shapeType=Capsule, length=0.505, radius=0.505, rotation=<rot 0 -90 0>}]
 dyn_trunks_string [{shapeType=Capsule, length=4.8912, radius=2.7052, shapeName=dyn_trunks_string_capsule, center=<3 0 2>, rotation=<rot 0 0 90>}]
 dyn_trunks_string_01 [{shapeType=Capsule, length=1.8494, radius=2.0474, shapeName=dyn_trunks_string_01_capsule, center=<3 0 1>, rotation=<rot 0 0 90>}]
 dyn_trunks_string_MSTR [{shapeType=Capsule, length=0.505, radius=0.505, rotation=<rot -3.6144 -90 -41.185>}]
 pelvis [{shapeType=Capsule, length=20.182, radius=8.9712, shapeName=pelvis_capsule, center=<0.2205 -0.4005 0.0217>, rotation=<rot 0.5275 -90 -0.6552>}]
 thigh_twist_01_l [{shapeType=Capsule, length=15.6744, radius=5.5037, center=<2.5341 0.224 0.6902>, rotation=<rot 90 0 0>}]
 thigh_twist_01_r [{shapeType=Capsule, length=15.6744, radius=5.5037, center=<-3.271 0.224 -0.7544>, rotation=<rot 90 0 0>}]
}
constraints [
 {bone1Name=dyn_trunks_rt_fr, bone2Name=dyn_trunks_rt_fr_MSTR, swing1LimitDegrees=7, swing2LimitDegrees=25, twistLimitDegrees=3}
 {bone1Name=dyn_trunks_rt_bk, bone2Name=dyn_trunks_rt_bk_MSTR, swing1LimitDegrees=7, swing2LimitDegrees=15, twistLimitDegrees=3}
 {bone1Name=dyn_trunks_lt_bk, bone2Name=dyn_trunks_lt_bk_MSTR, swing1LimitDegrees=7, swing2LimitDegrees=15, twistLimitDegrees=3}
 {bone1Name=dyn_trunks_lt_fr, bone2Name=dyn_trunks_lt_fr_MSTR, swing1LimitDegrees=7, swing2LimitDegrees=25, twistLimitDegrees=3}
 {bone1Name=dyn_trunks_string, bone2Name=dyn_trunks_string_MSTR, swing1LimitDegrees=15, swing2LimitDegrees=15, twistLimitDegrees=1}
 {
  bone1Name=thigh_twist_01_l
  bone2Name=pelvis
  swing1LimitDegrees=45
  swing1Motion=Free
  swing2LimitDegrees=45
  swing2Motion=Free
  twistLimitDegrees=45
  twistMotion=Free
 }
 {
  bone1Name=dyn_trunks_lt_fr_MSTR
  bone2Name=pelvis
  swing1LimitDegrees=45
  swing1Motion=Free
  swing2LimitDegrees=45
  swing2Motion=Free
  twistLimitDegrees=45
  twistMotion=Free
 }
 {
  bone1Name=dyn_trunks_lt_bk_MSTR
  bone2Name=pelvis
  swing1LimitDegrees=45
  swing1Motion=Free
  swing2LimitDegrees=45
  swing2Motion=Free
  twistLimitDegrees=45
  twistMotion=Free
 }
 {
  bone1Name=thigh_twist_01_r
  bone2Name=pelvis
  swing1LimitDegrees=45
  swing1Motion=Free
  swing2LimitDegrees=45
  swing2Motion=Free
  twistLimitDegrees=45
  twistMotion=Free
 }
 {
  bone1Name=dyn_trunks_rt_fr_MSTR
  bone2Name=pelvis
  swing1LimitDegrees=45
  swing1Motion=Free
  swing2LimitDegrees=45
  swing2Motion=Free
  twistLimitDegrees=45
  twistMotion=Free
 }
 {
  bone1Name=dyn_trunks_rt_bk_MSTR
  bone2Name=pelvis
  swing1LimitDegrees=45
  swing1Motion=Free
  swing2LimitDegrees=45
  swing2Motion=Free
  twistLimitDegrees=45
  twistMotion=Free
 }
 {
  bone1Name=dyn_trunks_string_MSTR
  bone2Name=pelvis
  swing1LimitDegrees=45
  swing1Motion=Free
  swing2LimitDegrees=45
  swing2Motion=Free
  twistLimitDegrees=45
  twistMotion=Free
 }
 {bone1Name=dyn_trunks_string_01, bone2Name=dyn_trunks_string, swing1LimitDegrees=40, swing2LimitDegrees=45, twistLimitDegrees=5}
]
