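# Physics asset for the crouch mesh of the male medium Alien Summer cosmetic body, covering dangling shirt, suit-back, trunks and drawstring bones.
# Capsules on pelvis, spine and thighs form the body volumes.
# Tiny MSTR anchor bodies link by free-rotation constraints to pelvis (shirt, string), spine_01 (suit) or thighs (trunks).
# Each dyn capsule hangs off its anchor with limited swing and twist, tightest on trunks; left front shirt has only an anchor.
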PhysicsAsset /BRCosmetics/Characters/Player/Male/Medium/Bodies/M_MED_Alien_Summer/Meshes/M_MED_Alien_Summer_crouch_Physics
bodies {
 dyn_shirt_01_l [{shapeType=Capsule, length=7.8175, radius=3.3499, center=<7.1401 0.0001 2>, rotation=<rot 0 0 90>}]
 dyn_shirt_01_l_MSTR [{shapeType=Capsule, length=0.505, radius=0.505, rotation=<rot 0 -90 0>}]
 dyn_shirt_01_r [{shapeType=Capsule, length=7.8175, radius=3.3499, center=<7.1401 0.0001 2>, rotation=<rot 0 0 90>}]
 dyn_shirt_01_r_MSTR [{shapeType=Capsule, length=0.505, radius=0.505, rotation=<rot 0 -90 0>}]
 dyn_shirt_fr_01_l_MSTR [{shapeType=Capsule, length=0.505, radius=0.505, rotation=<rot 0 -90 0>}]
 dyn_shirt_fr_01_r [{shapeType=Capsule, length=6.3761, radius=4.4025, center=<6.9158 0 2>, rotation=<rot 0 0 90>}]
 dyn_shirt_fr_01_r_MSTR [{shapeType=Capsule, length=0.505, radius=0.505, rotation=<rot 0 -90 0>}]
 dyn_suit_bk_lt [{shapeType=Capsule, length=9.0069, radius=4.0733, center=<9.9244 0 3.5>, rotation=<rot 0 0 79.9998>}]
 dyn_suit_bk_lt_MSTR [{shapeType=Capsule, length=0.505, radius=0.505, rotation=<rot 0 -90 0>}]
 dyn_suit_bk_rt [{shapeType=Capsule, length=9.0069, radius=4.0733, center=<9.9244 0 3.5>, rotation=<rot 0 0 -79.9998>}]
 dyn_suit_bk_rt_MSTR [{shapeType=Capsule, length=0.505, radius=0.505, rotation=<rot 0 -90 0>}]
 dyn_trunks_lt_bk [{shapeType=Capsule, length=4.3783, radius=2.8153, center=<11.2323 0 0>, rotation=<rot 0 0 90>}]
 dyn_trunks_lt_bk_MSTR [{shapeType=Capsule, length=0.505, radius=0.505, rotation=<rot 0 -90 0>}]
 dyn_trunks_lt_fr [{shapeType=Capsule, length=3.4545, radius=2.589, center=<11.2683 0 0>, rotation=<rot 0 0 90>}]
 dyn_trunks_lt_fr_MSTR [{shapeType=Capsule, length=0.505, radius=0.505, rotation=<rot 0 -90 0>}]
 dyn_trunks_rt_bk [{shapeType=Capsule, length=4.3783, radius=2.8153, center=<11.7042 0 0>, rotation=<rot 0 0 90>}]
 dyn_trunks_rt_bk_MSTR [{shapeType=Capsule, length=0.505, radius=0.505, rotation=<rot 0 -90 0>}]
 dyn_trunks_rt_fr [{shapeType=Capsule, length=3.4545, radius=2.589, center=<11.2683 0 0>, rotation=<rot 0 0 90>}]
 dyn_trunks_rt_fr_MSTR [{shapeType=Capsule, length=0.505, radius=0.505, rotation=<rot 0 -90 0>}]
 dyn_trunks_string [{shapeType=Capsule, length=3.2529, radius=1.7265, center=<4.1657 0 1>, rotation=<rot 0 0 90>}]
 dyn_trunks_string_MSTR [{shapeType=Capsule, length=0.505, radius=0.505, center=<0.416 0.254 -2.2041>, rotation=<rot -3.6144 -90 -41.185>}]
 pelvis [{shapeType=Capsule, length=8.6709, radius=9.4457, center=<-2.0733 -0.525 0>}]
 spine_01 [{shapeType=Capsule, length=8.2086, radius=8.0468, center=<0 -0.1776 0>}]
 thigh_l [{shapeType=Capsule, length=8.817, radius=8.9286, center=<-4.8287 0 2.3021>, rotation=<rot 85.2768 -179.9999 -179.9999>}]
 thigh_r [{shapeType=Capsule, length=8.817, radius=8.9286, center=<4.8287 0 -2.3021>, rotation=<rot 85.2768 -179.9999 -179.9999>}]
 thigh_twist_01_l [{shapeType=Capsule, length=15.6744, radius=5.5037, center=<2.5341 0.224 0.6902>, rotation=<rot 90 0 0>}]
 thigh_twist_01_r [{shapeType=Capsule, length=15.6744, radius=5.5037, center=<-3.271 0.224 -0.7544>, rotation=<rot 90 0 0>}]
}
constraints [
 {bone1Name=spine_01, bone2Name=pelvis, swing1LimitDegrees=45, swing2LimitDegrees=45, twistLimitDegrees=45}
 {
  bone1Name=thigh_l
  bone2Name=pelvis
  swing1LimitDegrees=45
  swing1Motion=Free
  swing2LimitDegrees=45
  swing2Motion=Free
  twistLimitDegrees=45
  twistMotion=Free
 }
 {
  bone1Name=thigh_r
  bone2Name=pelvis
  swing1LimitDegrees=45
  swing1Motion=Free
  swing2LimitDegrees=45
  swing2Motion=Free
  twistLimitDegrees=45
  twistMotion=Free
 }
 {
  bone1Name=dyn_shirt_fr_01_r_MSTR
  bone2Name=pelvis
  swing1LimitDegrees=45
  swing1Motion=Free
  swing2LimitDegrees=45
  swing2Motion=Free
  twistLimitDegrees=45
  twistMotion=Free
 }
 {
  bone1Name=dyn_shirt_01_r_MSTR
  bone2Name=pelvis
  swing1LimitDegrees=45
  swing1Motion=Free
  swing2LimitDegrees=45
  swing2Motion=Free
  twistLimitDegrees=45
  twistMotion=Free
 }
 {
  bone1Name=dyn_shirt_fr_01_l_MSTR
  bone2Name=pelvis
  swing1LimitDegrees=45
  swing1Motion=Free
  swing2LimitDegrees=45
  swing2Motion=Free
  twistLimitDegrees=45
  twistMotion=Free
 }
 {
  bone1Name=dyn_shirt_01_l_MSTR
  bone2Name=pelvis
  swing1LimitDegrees=45
  swing1Motion=Free
  swing2LimitDegrees=45
  swing2Motion=Free
  twistLimitDegrees=45
  twistMotion=Free
 }
 {
  bone1Name=dyn_trunks_string_MSTR
  bone2Name=pelvis
  swing1LimitDegrees=45
  swing1Motion=Free
  swing2LimitDegrees=45
  swing2Motion=Free
  twistLimitDegrees=45
  twistMotion=Free
 }
 {
  bone1Name=dyn_trunks_rt_bk_MSTR
  bone2Name=thigh_r
  swing1LimitDegrees=45
  swing1Motion=Free
  swing2LimitDegrees=45
  swing2Motion=Free
  twistLimitDegrees=45
  twistMotion=Free
 }
 {
  bone1Name=dyn_trunks_rt_fr_MSTR
  bone2Name=thigh_r
  swing1LimitDegrees=45
  swing1Motion=Free
  swing2LimitDegrees=45
  swing2Motion=Free
  twistLimitDegrees=45
  twistMotion=Free
 }
 {
  bone1Name=dyn_suit_bk_lt_MSTR
  bone2Name=spine_01
  swing1LimitDegrees=45
  swing1Motion=Free
  swing2LimitDegrees=45
  swing2Motion=Free
  twistLimitDegrees=45
  twistMotion=Free
 }
 {
  bone1Name=dyn_suit_bk_rt_MSTR
  bone2Name=spine_01
  swing1LimitDegrees=45
  swing1Motion=Free
  swing2LimitDegrees=45
  swing2Motion=Free
  twistLimitDegrees=45
  twistMotion=Free
 }
 {bone1Name=dyn_suit_bk_rt, bone2Name=dyn_suit_bk_rt_MSTR, swing1LimitDegrees=15, swing2LimitDegrees=45, twistLimitDegrees=1}
 {bone1Name=dyn_suit_bk_lt, bone2Name=dyn_suit_bk_lt_MSTR, swing1LimitDegrees=15, swing2LimitDegrees=45, twistLimitDegrees=1}
 {bone1Name=dyn_trunks_rt_fr, bone2Name=dyn_trunks_rt_fr_MSTR, swing1LimitDegrees=3, swing2LimitDegrees=8, twistLimitDegrees=1}
 {bone1Name=dyn_trunks_rt_bk, bone2Name=dyn_trunks_rt_bk_MSTR, swing1LimitDegrees=7, swing2LimitDegrees=15, twistLimitDegrees=3}
 {bone1Name=dyn_trunks_string, bone2Name=dyn_trunks_string_MSTR, swing1LimitDegrees=20, swing2LimitDegrees=45, twistLimitDegrees=5}
 {bone1Name=dyn_shirt_01_l, bone2Name=dyn_shirt_01_l_MSTR, swing1LimitDegrees=35, swing2LimitDegrees=60, twistLimitDegrees=5}
 {bone1Name=dyn_shirt_01_r, bone2Name=dyn_shirt_01_r_MSTR, swing1LimitDegrees=35, swing2LimitDegrees=60, twistLimitDegrees=5}
 {bone1Name=dyn_shirt_fr_01_r, bone2Name=dyn_shirt_fr_01_r_MSTR, swing1LimitDegrees=35, swing2LimitDegrees=60, twistLimitDegrees=5}
 {bone1Name=dyn_trunks_lt_bk, bone2Name=dyn_trunks_lt_bk_MSTR, swing1LimitDegrees=7, swing2LimitDegrees=15, twistLimitDegrees=3}
 {
  bone1Name=dyn_trunks_lt_bk_MSTR
  bone2Name=thigh_l
  swing1LimitDegrees=45
  swing1Motion=Free
  swing2LimitDegrees=45
  swing2Motion=Free
  twistLimitDegrees=45
  twistMotion=Free
 }
 {bone1Name=dyn_trunks_lt_fr, bone2Name=dyn_trunks_lt_fr_MSTR, swing1LimitDegrees=3, swing2LimitDegrees=8, twistLimitDegrees=1}
 {
  bone1Name=dyn_trunks_lt_fr_MSTR
  bone2Name=thigh_l
  swing1LimitDegrees=45
  swing1Motion=Free
  swing2LimitDegrees=45
  swing2Motion=Free
  twistLimitDegrees=45
  twistMotion=Free
 }
 {
  bone1Name=thigh_twist_01_l
  bone2Name=thigh_l
  swing1LimitDegrees=45
  swing1Motion=Free
  swing2LimitDegrees=45
  swing2Motion=Free
  twistLimitDegrees=45
  twistMotion=Free
 }
 {
  bone1Name=thigh_twist_01_r
  bone2Name=thigh_r
  swing1LimitDegrees=45
  swing1Motion=Free
  swing2LimitDegrees=45
  swing2Motion=Free
  twistLimitDegrees=45
  twistMotion=Free
 }
]
