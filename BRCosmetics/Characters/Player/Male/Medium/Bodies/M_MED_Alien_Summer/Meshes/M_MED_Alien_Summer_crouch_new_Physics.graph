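# Physics asset for the crouch mesh of the male medium Alien Summer body, rigging its shirt and back suit pieces for secondary motion.
# Capsules on pelvis, spine_01 and thighs form the body volume.
# Tiny MSTR anchor bodies join shirt pieces to pelvis and suit pieces to spine_01 with free angular motion.
# Each dyn capsule hangs from its anchor with angular limits, tighter on the suit's swing1 and twist than the shirt's.
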PhysicsAsset /BRCosmetics/Characters/Player/Male/Medium/Bodies/M_MED_Alien_Summer/Meshes/M_MED_Alien_Summer_crouch_new_Physics
bodies {
 dyn_shirt_01_l [{shapeType=Capsule, length=7.8175, radius=3.3499, center=<7.1401 0.0001 2>, rotation=<rot 0 0 90>}]
 dyn_shirt_01_l_MSTR [{shapeType=Capsule, length=0.505, radius=0.505, rotation=<rot 0 -90 0>}]
 dyn_shirt_01_r [{shapeType=Capsule, length=7.99, radius=3.1773, center=<7.1401 0.0001 2>, rotation=<rot 0 0 90>}]
 dyn_shirt_01_r_MSTR [{shapeType=Capsule, length=0.505, radius=0.505, rotation=<rot 0 -90 0>}]
 dyn_shirt_fr_01_l [{shapeType=Capsule, length=7.4489, radius=3.3297, center=<6.9158 0 1.5>, rotation=<rot 0 0 95.0001>}]
 dyn_shirt_fr_01_l_MSTR [{shapeType=Capsule, length=0.505, radius=0.505, rotation=<rot 0 -90 0>}]
 dyn_shirt_fr_01_r [{shapeType=Capsule, length=7.3155, radius=3.4631, center=<6.9158 0 2>, rotation=<rot 0 0 90>}]
 dyn_shirt_fr_01_r_MSTR [{shapeType=Capsule, length=0.505, radius=0.505, rotation=<rot 0 -90 0>}]
 dyn_suit_bk_lt [{shapeType=Capsule, length=9.0069, radius=4.0733, center=<9.9244 0 3.5>, rotation=<rot 0 0 79.9998>}]
 dyn_suit_bk_lt_MSTR [{shapeType=Capsule, length=0.505, radius=0.505, rotation=<rot 0 -90 0>}]
 dyn_suit_bk_rt [{shapeType=Capsule, length=8.7983, radius=4.2819, center=<9.9244 0 3.5>, rotation=<rot 0 0 -79.9998>}]
 dyn_suit_bk_rt_MSTR [{shapeType=Capsule, length=0.505, radius=0.505, rotation=<rot 0 -90 0>}]
 pelvis [{shapeType=Capsule, length=8.9482, radius=9.1683, center=<-2.0733 -0.8042 -0.0001>}]
 spine_01 [{shapeType=Capsule, length=8.2086, radius=8.0468, center=<0 -0.1776 0>}]
 thigh_l [{shapeType=Capsule, length=8.5281, radius=8.9719, center=<-4.8287 0 2.3021>, rotation=<rot 85.2768 -179.9999 -179.9999>}]
 thigh_r [{shapeType=Capsule, length=8.4179, radius=9.0821, center=<4.8287 0 -2.3021>, rotation=<rot 85.2768 -179.9999 -179.9999>}]
}
constraints [
 {bone1Name=spine_01, bone2Name=pelvis, swing1LimitDegrees=45, swing2LimitDegrees=45, twistLimitDegrees=45}
 {
  bone1Name=thigh_l
  bone2Name=pelvis
  swing1LimitDegrees=45
  swing1Motion=Free
  swing2LimitDegrees=45
  swing2Motion=Free
  twistLimitDegrees=45
  twistMotion=Free
 }
 {
  bone1Name=thigh_r
  bone2Name=pelvis
  swing1LimitDegrees=45
  swing1Motion=Free
  swing2LimitDegrees=45
  swing2Motion=Free
  twistLimitDegrees=45
  twistMotion=Free
 }
 {
  bone1Name=dyn_shirt_fr_01_r_MSTR
  bone2Name=pelvis
  swing1LimitDegrees=45
  swing1Motion=Free
  swing2LimitDegrees=45
  swing2Motion=Free
  twistLimitDegrees=45
  twistMotion=Free
 }
 {
  bone1Name=dyn_shirt_01_r_MSTR
  bone2Name=pelvis
  swing1LimitDegrees=45
  swing1Motion=Free
  swing2LimitDegrees=45
  swing2Motion=Free
  twistLimitDegrees=45
  twistMotion=Free
 }
 {
  bone1Name=dyn_shirt_fr_01_l_MSTR
  bone2Name=pelvis
  swing1LimitDegrees=45
  swing1Motion=Free
  swing2LimitDegrees=45
  swing2Motion=Free
  twistLimitDegrees=45
  twistMotion=Free
 }
 {
  bone1Name=dyn_shirt_01_l_MSTR
  bone2Name=pelvis
  swing1LimitDegrees=45
  swing1Motion=Free
  swing2LimitDegrees=45
  swing2Motion=Free
  twistLimitDegrees=45
  twistMotion=Free
 }
 {
  bone1Name=dyn_suit_bk_lt_MSTR
  bone2Name=spine_01
  swing1LimitDegrees=45
  swing1Motion=Free
  swing2LimitDegrees=45
  swing2Motion=Free
  twistLimitDegrees=45
  twistMotion=Free
 }
 {
  bone1Name=dyn_suit_bk_rt_MSTR
  bone2Name=spine_01
  swing1LimitDegrees=45
  swing1Motion=Free
  swing2LimitDegrees=45
  swing2Motion=Free
  twistLimitDegrees=45
  twistMotion=Free
 }
 {bone1Name=dyn_suit_bk_rt, bone2Name=dyn_suit_bk_rt_MSTR, swing1LimitDegrees=10, swing2LimitDegrees=45, twistLimitDegrees=3}
 {bone1Name=dyn_suit_bk_lt, bone2Name=dyn_suit_bk_lt_MSTR, swing1LimitDegrees=10, swing2LimitDegrees=45, twistLimitDegrees=3}
 {bone1Name=dyn_shirt_01_l, bone2Name=dyn_shirt_01_l_MSTR, swing1LimitDegrees=35, swing2LimitDegrees=60, twistLimitDegrees=5}
 {bone1Name=dyn_shirt_fr_01_l, bone2Name=dyn_shirt_fr_01_l_MSTR, swing1LimitDegrees=35, swing2LimitDegrees=60, twistLimitDegrees=5}
 {bone1Name=dyn_shirt_01_r, bone2Name=dyn_shirt_01_r_MSTR, swing1LimitDegrees=35, swing2LimitDegrees=45, twistLimitDegrees=5}
 {bone1Name=dyn_shirt_fr_01_r, bone2Name=dyn_shirt_fr_01_r_MSTR, swing1LimitDegrees=35, swing2LimitDegrees=45, twistLimitDegrees=5}
]
