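# Animation Blueprint for the male M_MED_Alien_Summer cosmetic body, adding deformation and secondary motion to the parent mesh's copied pose.
# Deformation and emote Control Rigs, slope-foot bones, thigh-following shirt constraints and optional torso scaling produce a cached pre-simulation pose.
# Variant-tag-driven Shirt Active selects between crouch-switched RigidBody physics and an AnimDynamics coattail chain with RigidBody, both using thigh-rotation corrective offsets.
# A final RigidBody pass simulates the trunks.
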AnimBlueprint /BRCosmetics/Characters/Player/Male/Medium/Bodies/M_MED_Alien_Summer/Meshes/M_MED_Alien_Summer_AnimBP
target_skeleton=&/M_MED_Alien_Summer_Skeleton
graphs [
 AnimGraph {nodes [
   n1 = "Constraint - dyn_shirt_fr_01_l_MSTR (RR)"!(ComponentPose, Alpha=1, ConstraintWeights_0=0.1, ConstraintWeights_1:float=0.9) -> (Pose) {
    alphaBoolBlend {blendOption=Linear}
    boneToModify {boneName=dyn_shirt_fr_01_l_MSTR}
    constraintSetup [{transformType=Rotation, targetBone {boneName=pelvis}}, {transformType=Rotation, targetBone {boneName=thigh_l}}]
    constraintWeights [1,1]
   }
   n2 = "Constraint - dyn_shirt_01_l_MSTR (RR)"!(ComponentPose=@n1, Alpha=1, ConstraintWeights_0=0.4, ConstraintWeights_1:float=0.6) -> (Pose) {
    alphaBoolBlend {blendOption=Linear}
    boneToModify {boneName=dyn_shirt_01_l_MSTR}
    constraintSetup [{transformType=Rotation, targetBone {boneName=pelvis}}, {transformType=Rotation, targetBone {boneName=thigh_l}}]
    constraintWeights [1,1]
   }
   n3 = "Constraint - dyn_shirt_fr_01_r_MSTR (RR)"!(ComponentPose=@n2, Alpha=1, ConstraintWeights_0=0.1, ConstraintWeights_1:float=0.9) -> (Pose) {
    alphaBoolBlend {blendOption=Linear}
    boneToModify {boneName=dyn_shirt_fr_01_r_MSTR}
    constraintSetup [{transformType=Rotation, targetBone {boneName=pelvis}}, {transformType=Rotation, targetBone {boneName=thigh_r}}]
    constraintWeights [1,1]
   }
   "Constraint - dyn_shirt_01_r_MSTR (RR)"!(ComponentPose=@n3, Alpha=1, ConstraintWeights_0=0.4, ConstraintWeights_1:float=0.6) -> (Pose) {
    alphaBoolBlend {blendOption=Linear}
    boneToModify {boneName=dyn_shirt_01_r_MSTR}
    constraintSetup [{transformType=Rotation, targetBone {boneName=pelvis}}, {transformType=Rotation, targetBone {boneName=thigh_r}}]
    constraintWeights [1,1]
   }
   Local To Component(LocalPose=Component To Local())
   "Transform (Modify) Bone - Bone: dyn_shirt_fr_01_l_MSTR"(Alpha=1, Translation="-2,-2,7", Rotation="0,-20,20", Scale="1,1,1") {
    rotationMode=BMM_Additive
    rotationSpace=BCS_BoneSpace
    translationMode=BMM_Additive
    translationSpace=BCS_BoneSpace
    alphaBoolBlend {blendOption=Linear}
    alphaScaleBiasClamp {bInterpResult=true, interpSpeedDecreasing=15, interpSpeedIncreasing=40}
    boneToModify {boneName=dyn_shirt_fr_01_l_MSTR}
    rotation {roll=20, yaw=-20}
    translation=<-2 -2 7>
   }
   "Transform (Modify) Bone - Bone: dyn_shirt_fr_01_l_MSTR"(Alpha=1, Translation="2,0.662,3.710", Rotation="0,-30,-45", Scale="1,1,1") {
    rotationMode=BMM_Additive
    translationMode=BMM_Additive
    alphaBoolBlend {blendOption=Linear}
    alphaScaleBiasClamp {bInterpResult=true, bMapRange=true, interpSpeedDecreasing=5, interpSpeedIncreasing=30}
    boneToModify {boneName=dyn_shirt_fr_01_l_MSTR}
    rotation {roll=-45, yaw=-30}
    translation=<2 0.662 3.71>
   }
   n4 = "Transform (Modify) Bone - Bone: dyn_shirt_fr_01_r_MSTR"(Alpha=1, Translation="-1,1,5", Rotation="0,0,0", Scale="1,1,1") {
    rotationMode=BMM_Additive
    rotationSpace=BCS_BoneSpace
    translationMode=BMM_Additive
    translationSpace=BCS_BoneSpace
    alphaBoolBlend {blendOption=Linear}
    alphaScaleBiasClamp {bInterpResult=true, interpSpeedDecreasing=15, interpSpeedIncreasing=40}
    boneToModify {boneName=dyn_shirt_fr_01_r_MSTR}
    translation=<-1 1 5>
   }
   "Transform (Modify) Bone - Bone: dyn_shirt_01_r_MSTR"(ComponentPose=@n4, Alpha=1, Translation="-1,-2,3", Rotation="0,0,0", Scale="1,1,1") {
    rotationMode=BMM_Additive
    rotationSpace=BCS_BoneSpace
    translationMode=BMM_Additive
    translationSpace=BCS_BoneSpace
    alphaBoolBlend {blendOption=Linear}
    alphaScaleBiasClamp {bInterpResult=true, interpSpeedDecreasing=15, interpSpeedIncreasing=40}
    boneToModify {boneName=dyn_shirt_01_r_MSTR}
    translation=<-1 -2 3>
   }
   "Transform (Modify) Bone - Bone: dyn_shirt_fr_01_l_MSTR"(Alpha=1, Translation="-3.000,1,7", Rotation="10.000,-20,30", Scale="1,1,1") {
    rotationMode=BMM_Additive
    rotationSpace=BCS_BoneSpace
    translationMode=BMM_Additive
    translationSpace=BCS_BoneSpace
    alphaBoolBlend {blendOption=Linear}
    alphaScaleBiasClamp {bInterpResult=true, interpSpeedDecreasing=15, interpSpeedIncreasing=40}
    boneToModify {boneName=dyn_shirt_fr_01_l_MSTR}
    rotation=<rot 10 -20 30>
    translation=<-3 1 7>
   }
   n5 = "Transform (Modify) Bone - Bone: dyn_shirt_fr_01_r_MSTR"(Alpha=1, Translation="-2.679,1.882,4.458", Rotation="0,-20,30", Scale="1,1,1") {
    rotationMode=BMM_Additive
    rotationSpace=BCS_BoneSpace
    translationMode=BMM_Additive
    translationSpace=BCS_BoneSpace
    alphaBoolBlend {blendOption=Linear}
    alphaScaleBiasClamp {bInterpResult=true, interpSpeedDecreasing=15, interpSpeedIncreasing=40}
    boneToModify {boneName=dyn_shirt_fr_01_r_MSTR}
    rotation {roll=30, yaw=-20}
    translation=<-2.679 1.882 4.458>
   }
   "Transform (Modify) Bone - Bone: dyn_shirt_01_r_MSTR"(ComponentPose=@n5, Alpha=1, Translation="-1,-0.000,2.000", Rotation="0,0,0", Scale="1,1,1") {
    rotationMode=BMM_Additive
    rotationSpace=BCS_BoneSpace
    translationMode=BMM_Additive
    translationSpace=BCS_BoneSpace
    alphaBoolBlend {blendOption=Linear}
    alphaScaleBiasClamp {bInterpResult=true, interpSpeedDecreasing=15, interpSpeedIncreasing=40}
    boneToModify {boneName=dyn_shirt_01_r_MSTR}
    translation {x=-1, z=2}
   }
   "Transform (Modify) Bone - Bone: dyn_shirt_fr_01_l_MSTR"(Alpha=1, Translation="0.670,-0.213,10.989", Rotation="0,-25,-30", Scale="1,1,1") {
    rotationMode=BMM_Additive
    scaleMode=BMM_Additive
    scaleSpace=BCS_BoneSpace
    translationMode=BMM_Additive
    alphaBoolBlend {blendOption=Linear}
    alphaScaleBiasClamp {inRange {max=0, min=-100}, outRange {max=0, min=1}}
    boneToModify {boneName=dyn_shirt_fr_01_l_MSTR}
    rotation {roll=-30, yaw=-25}
    translation=<0.67 -0.213 10.989>
   }
   "Transform (Modify) Bone - Bone: dyn_shirt_01_l_MSTR"(Alpha=1, Translation="2,0,0", Rotation="0,0,0", Scale="1,1,1") {
    rotationMode=BMM_Additive
    scaleMode=BMM_Additive
    scaleSpace=BCS_BoneSpace
    translationMode=BMM_Additive
    alphaBoolBlend {blendOption=Linear}
    alphaScaleBiasClamp {bInterpResult=true, inRange {max=0, min=-100}, outRange {max=0, min=1}}
    boneToModify {boneName=dyn_shirt_01_l_MSTR}
    translation {x=2}
   }
   "Transform (Modify) Bone - Bone: dyn_shirt_01_l_MSTR"(Alpha=1, Translation="2,0,0", Rotation="0,0,0", Scale="1,1,1") {
    rotationMode=BMM_Additive
    translationMode=BMM_Additive
    alphaBoolBlend {blendOption=Linear}
    alphaScaleBiasClamp {bInterpResult=true, interpSpeedDecreasing=5, interpSpeedIncreasing=30}
    boneToModify {boneName=dyn_shirt_01_l_MSTR}
    translation {x=2}
   }
   n6 = "Transform (Modify) Bone - Bone: dyn_shirt_01_r_MSTR"!(ComponentPose, bAlphaBoolEnabled:bool=True, Translation="-3,-3,3", Rotation="0,0,0", Scale="1,1,1") -> (Pose) {
    alphaInputType=Bool
    rotationMode=BMM_Additive
    rotationSpace=BCS_BoneSpace
    scaleMode=BMM_Replace
    scaleSpace=BCS_BoneSpace
    translationMode=BMM_Additive
    translationSpace=BCS_BoneSpace
    alphaBoolBlend {blendInTime=0.2, blendOption=Linear, blendOutTime=0.2}
    alphaScaleBiasClamp {bClampResult=true, bInterpResult=true}
    boneToModify {boneName=dyn_shirt_01_r_MSTR}
    translation=<-3 -3 3>
   }
   n7 = Get thigh_rot_l()
   n8 = "Transform (Modify) Bone - Bone: dyn_shirt_fr_01_l_MSTR"(ComponentPose=@n6, Alpha=@n7, Translation="0.000,0.2,3", Rotation="10.000,0.000,35", Scale="1,1,1") {
    rotationMode=BMM_Additive
    translationMode=BMM_Additive
    alphaBoolBlend {blendInTime=0.1, blendOption=Linear, blendOutTime=0.4}
    boneToModify {boneName=dyn_shirt_fr_01_l_MSTR}
    rotation {pitch=10, roll=35}
    translation {y=0.2, z=3}
   }
   "Transform (Modify) Bone - Bone: dyn_shirt_01_l_MSTR"(ComponentPose=@n8, Alpha=@n7, Translation="0,0,0", Rotation="0.000,0.000,35.000", Scale="1,1,1") {
    rotationMode=BMM_Additive
    translationMode=BMM_Additive
    alphaBoolBlend {blendInTime=0.1, blendOption=Linear, blendOutTime=0.4}
    boneToModify {boneName=dyn_shirt_01_l_MSTR}
    rotation {roll=35}
   }
   "Transform (Modify) Bone - Bone: dyn_shirt_fr_01_r_MSTR"!(ComponentPose, bAlphaBoolEnabled:bool=Get bIsJumping(), Translation="-2,0.000,5", Rotation="0.000,30,-40", Scale="1,1,1") -> (Pose) {
    alphaInputType=Bool
    rotationMode=BMM_Additive
    translationMode=BMM_Additive
    alphaBoolBlend {blendInTime=0.1, blendOption=Linear, blendOutTime=0.4}
    boneToModify {boneName=dyn_shirt_fr_01_r_MSTR}
    rotation {roll=-40, yaw=30}
    translation {x=-2, z=5}
   }
   n9 = "Transform (Modify) Bone - Bone: dyn_shirt_fr_01_r_MSTR"!(ComponentPose, Alpha:float=1, Translation="-2,0,5", Rotation="0,20,-30", Scale="1,1,1") -> (Pose) {
    rotationMode=BMM_Additive
    rotationSpace=BCS_BoneSpace
    translationMode=BMM_Additive
    translationSpace=BCS_BoneSpace
    alphaBoolBlend {blendOption=Linear}
    alphaScaleBiasClamp {bInterpResult=true, interpSpeedDecreasing=15, interpSpeedIncreasing=30}
    boneToModify {boneName=dyn_shirt_fr_01_r_MSTR}
    rotation {roll=-30, yaw=20}
    translation {x=-2, z=5}
   }
   "Transform (Modify) Bone - Bone: dyn_shirt_01_r_MSTR"!(ComponentPose=@n9, Alpha:float=1, Translation="-3,-3,3", Rotation="0,0,0", Scale="1,1,1") -> (Pose) {
    rotationMode=BMM_Additive
    rotationSpace=BCS_BoneSpace
    translationMode=BMM_Additive
    translationSpace=BCS_BoneSpace
    alphaBoolBlend {blendOption=Linear}
    alphaScaleBiasClamp {bInterpResult=true, interpSpeedDecreasing=15, interpSpeedIncreasing=30}
    boneToModify {boneName=dyn_shirt_01_r_MSTR}
    translation=<-3 -3 3>
   }
   RigidBody() {
    bClampLinearTranslationLimitToRefPose=true
    bOverrideWorldGravity=true
    simulationSpace=BaseBoneSpace
    alphaBoolBlend {blendOption=Linear}
    baseBoneRef {boneName=thigh_r}
    componentLinearAccScale=<0.3 0.3 0.3>
    componentLinearVelScale=<0.3 0.3 0.3>
    overrideWorldGravity {z=-980}
   }
   n10 = Get bIsCrouching()
   RigidBody!(ComponentPose, bAlphaBoolEnabled:bool=@n10, ExternalForce) -> (Pose) {
    alphaInputType=Bool
    bClampLinearTranslationLimitToRefPose=true
    bOverrideWorldGravity=true
    overridePhysicsAsset=&/M_MED_Alien_Summer_crouch_Physics
    simulationSpace=BaseBoneSpace
    alphaBoolBlend {blendInTime=0.1, blendOption=Linear, blendOutTime=0.1}
    baseBoneRef {boneName=pelvis}
    componentAppliedLinearAccClamp=<3000 3000 3000>
    componentLinearAccScale=<0.3 0.3 0.5>
    componentLinearVelScale=<0.3 0.3 0.5>
    overrideWorldGravity {z=-980}
   }
   n11 = Get thigh_rot_l()
   n12 = "Transform (Modify) Bone - Bone: dyn_shirt_fr_01_l_MSTR"(
    ComponentPose=Local To Component(LocalPose=Use cached pose 'presim'())
    Alpha=@n11
    Translation="-2,-2,7"
    Rotation="30,10,20"
    Scale="1,1,1"
   ) {
    rotationMode=BMM_Additive
    rotationSpace=BCS_BoneSpace
    translationMode=BMM_Additive
    translationSpace=BCS_BoneSpace
    alphaBoolBlend {blendOption=Linear}
    alphaScaleBiasClamp {bInterpResult=true, interpSpeedDecreasing=40, interpSpeedIncreasing=40}
    boneToModify {boneName=dyn_shirt_fr_01_l_MSTR}
    rotation=<rot 30 10 20>
    translation=<-2 -2 7>
   }
   n13 = "Transform (Modify) Bone - Bone: dyn_shirt_01_r_MSTR"(ComponentPose=@n12, Alpha=@n11, Translation="-3,-3,3", Rotation="0,0,0", Scale="1,1,1") {
    rotationMode=BMM_Additive
    rotationSpace=BCS_BoneSpace
    translationMode=BMM_Additive
    translationSpace=BCS_BoneSpace
    alphaBoolBlend {blendOption=Linear}
    alphaScaleBiasClamp {bInterpResult=true, interpSpeedDecreasing=40, interpSpeedIncreasing=40}
    boneToModify {boneName=dyn_shirt_01_r_MSTR}
    translation=<-3 -3 3>
   }
   n14 = Get thigh_rot_r()
   n15 = "Transform (Modify) Bone - Bone: dyn_shirt_fr_01_r_MSTR"(ComponentPose=@n13, Alpha=@n14, Translation="-1,0.5,5", Rotation="15,0,-30", Scale="1,1,1") {
    rotationMode=BMM_Additive
    rotationSpace=BCS_BoneSpace
    scaleMode=BMM_Additive
    scaleSpace=BCS_BoneSpace
    translationMode=BMM_Additive
    translationSpace=BCS_BoneSpace
    alphaBoolBlend {blendOption=Linear}
    alphaScaleBiasClamp {bInterpResult=true, interpSpeedDecreasing=40, interpSpeedIncreasing=40, inRange {max=0, min=-100}, outRange {max=0, min=1}}
    boneToModify {boneName=dyn_shirt_fr_01_r_MSTR}
    rotation {pitch=15, roll=-30}
    translation=<-1 0.5 5>
   }
   n16 = "Transform (Modify) Bone - Bone: dyn_shirt_01_r_MSTR"(ComponentPose=@n15, Alpha=@n14, Translation="0,0,1", Rotation="0,0,0", Scale="1,1,1") {
    rotationMode=BMM_Additive
    rotationSpace=BCS_BoneSpace
    scaleMode=BMM_Additive
    scaleSpace=BCS_BoneSpace
    translationMode=BMM_Additive
    translationSpace=BCS_BoneSpace
    alphaBoolBlend {blendOption=Linear}
    alphaScaleBiasClamp {bInterpResult=true, interpSpeedDecreasing=40, interpSpeedIncreasing=40, inRange {max=0, min=-100}, outRange {max=0, min=1}}
    boneToModify {boneName=dyn_shirt_01_r_MSTR}
    translation {z=1}
   }
   n17 = Get bIsCrouching()
   n18 = RigidBody(ComponentPose=@n16, bAlphaBoolEnabled=NOT Boolean(A=@n17)) {
    alphaInputType=Bool
    bClampLinearTranslationLimitToRefPose=true
    bOverrideWorldGravity=true
    overridePhysicsAsset=&/M_MED_Alien_Summer_Physics
    simulationSpace=BaseBoneSpace
    alphaBoolBlend {blendInTime=0.2, blendOption=Linear, blendOutTime=0.2}
    baseBoneRef {boneName=pelvis}
    componentAppliedLinearAccClamp=<3000 3000 3000>
    componentLinearAccScale=<0.4 0.4 0.5>
    componentLinearVelScale=<0.25 0.25 0.35>
    overrideWorldGravity {z=-980}
   }
   n19 = RigidBody(ComponentPose=@n18, bAlphaBoolEnabled=@n17) {
    alphaInputType=Bool
    bClampLinearTranslationLimitToRefPose=true
    bOverrideWorldGravity=true
    overridePhysicsAsset=&/M_MED_Alien_Summer_crouch_new_Physics
    simulationSpace=BaseBoneSpace
    alphaBoolBlend {blendInTime=0.2, blendOption=Linear, blendOutTime=0.2}
    baseBoneRef {boneName=pelvis}
    componentAppliedLinearAccClamp=<3000 3000 3000>
    componentLinearAccScale=<0.4 0.4 0.5>
    componentLinearVelScale=<0.25 0.25 0.35>
    overrideWorldGravity {z=-980}
   }
   n20 = Component To Local(ComponentPose="Transform (Modify) Bone - Bone: dyn_belt"(ComponentPose=@n19, Alpha=1, Translation="0,0,0", Rotation="0,0,0", Scale="1,1,0.75") {scaleMode=BMM_Additive, alphaBoolBlend {blendOption=Linear}, boneToModify {boneName=dyn_belt}, scale {z=0.75}})
   n21 = Get thigh_rot_l()
   n22 = "Transform (Modify) Bone - Bone: dyn_shirt_fr_01_l_MSTR"(
    ComponentPose=Local To Component(LocalPose=Use cached pose 'presim'())
    Alpha=@n21
    Translation="-2,-2,7"
    Rotation="30,10,20"
    Scale="1,1,1"
   ) {
    rotationMode=BMM_Additive
    rotationSpace=BCS_BoneSpace
    translationMode=BMM_Additive
    translationSpace=BCS_BoneSpace
    alphaBoolBlend {blendOption=Linear}
    alphaScaleBiasClamp {bInterpResult=true, interpSpeedDecreasing=15, interpSpeedIncreasing=40}
    boneToModify {boneName=dyn_shirt_fr_01_l_MSTR}
    rotation=<rot 30 10 20>
    translation=<-2 -2 7>
   }
   n23 = "Transform (Modify) Bone - Bone: dyn_shirt_01_r_MSTR"(ComponentPose=@n22, Alpha=@n21, Translation="-3,-3,3", Rotation="0,0,0", Scale="1,1,1") {
    rotationMode=BMM_Additive
    rotationSpace=BCS_BoneSpace
    translationMode=BMM_Additive
    translationSpace=BCS_BoneSpace
    alphaBoolBlend {blendOption=Linear}
    alphaScaleBiasClamp {bInterpResult=true, interpSpeedDecreasing=15, interpSpeedIncreasing=40}
    boneToModify {boneName=dyn_shirt_01_r_MSTR}
    translation=<-3 -3 3>
   }
   n24 = Get thigh_rot_r()
   n25 = "Transform (Modify) Bone - Bone: dyn_shirt_fr_01_r_MSTR"(ComponentPose=@n23, Alpha=@n24, Translation="-2,0,7", Rotation="30,0,-30", Scale="1,1,1") {
    rotationMode=BMM_Additive
    rotationSpace=BCS_BoneSpace
    scaleMode=BMM_Additive
    scaleSpace=BCS_BoneSpace
    translationMode=BMM_Additive
    translationSpace=BCS_BoneSpace
    alphaBoolBlend {blendOption=Linear}
    alphaScaleBiasClamp {bInterpResult=true, interpSpeedDecreasing=15, interpSpeedIncreasing=40, inRange {max=0, min=-100}, outRange {max=0, min=1}}
    boneToModify {boneName=dyn_shirt_fr_01_r_MSTR}
    rotation {pitch=30, roll=-30}
    translation {x=-2, z=7}
   }
   n26 = "Transform (Modify) Bone - Bone: dyn_shirt_01_r_MSTR"(ComponentPose=@n25, Alpha=@n24, Translation="0,0,3", Rotation="0,0,0", Scale="1,1,1") {
    rotationMode=BMM_Additive
    rotationSpace=BCS_BoneSpace
    scaleMode=BMM_Additive
    scaleSpace=BCS_BoneSpace
    translationMode=BMM_Additive
    translationSpace=BCS_BoneSpace
    alphaBoolBlend {blendOption=Linear}
    alphaScaleBiasClamp {bInterpResult=true, interpSpeedDecreasing=15, interpSpeedIncreasing=40, inRange {max=0, min=-100}, outRange {max=0, min=1}}
    boneToModify {boneName=dyn_shirt_01_r_MSTR}
    translation {z=3}
   }
   n27 = "Transform (Modify) Bone - Bone: dyn_shirt_fr_01_l_MSTR"!(ComponentPose=@n26, bAlphaBoolEnabled:bool=Get bIsCrouching(), Translation="0,0,1", Rotation="0,0,0", Scale="1,1,1") -> (Pose) {
    alphaInputType=Bool
    rotationMode=BMM_Additive
    rotationSpace=BCS_BoneSpace
    scaleMode=BMM_Additive
    scaleSpace=BCS_BoneSpace
    translationMode=BMM_Additive
    translationSpace=BCS_BoneSpace
    alphaBoolBlend {blendInTime=0.1, blendOption=Linear, blendOutTime=0.1}
    alphaScaleBiasClamp {bClampResult=true, bInterpResult=true}
    boneToModify {boneName=dyn_shirt_fr_01_l_MSTR}
    translation {z=1}
   }
   n28 = Anim Dynamics - Bone: dyn_suit_bk_rt(ComponentPose=@n27, Alpha=1, ExternalForce="0,0,0") {
    angularDampingOverride=0.9999
    angularSpringConstant=100
    bAngularSpring=true
    bGravityOverrideInSimSpace=true
    bOverrideAngularDamping=true
    bUseGravityOverride=true
    simulationSpace=RootRelative
    alphaBoolBlend {blendOption=Linear}
    boundBone {boneName=dyn_suit_bk_rt}
    componentAppliedLinearAccClamp=<3500 3500 3500>
    componentLinearAccScale=<0.2 0.2 0.4>
    componentLinearVelScale=<0.5 0.5 0.5>
    gravityOverride {z=-980}
    physicsBodyDefinitions [{
      sphereCollisionRadius=0
      boundBone {boneName=dyn_suit_bk_rt}
      boxExtents=<10 20 20>
      constraintSetup {angularLimitsMax=<1 -10 5>, angularLimitsMin=<-1 -45 -5>, angularTarget=<0 0 0>}
      localJointOffset=<2 0 0>
     }]
    relativeSpaceBone {boneName=root}
   }
   n29 = Anim Dynamics - Bone: dyn_suit_bk_lt(ComponentPose=@n28, Alpha=1, ExternalForce="0,0,0") {
    angularDampingOverride=0.9999
    angularSpringConstant=100
    bAngularSpring=true
    bGravityOverrideInSimSpace=true
    bOverrideAngularDamping=true
    bUseGravityOverride=true
    simulationSpace=RootRelative
    alphaBoolBlend {blendOption=Linear}
    boundBone {boneName=dyn_suit_bk_lt}
    componentAppliedLinearAccClamp=<3500 3500 3500>
    componentLinearAccScale=<0.2 0.2 0.4>
    componentLinearVelScale=<0.5 0.5 0.5>
    gravityOverride {z=-980}
    physicsBodyDefinitions [{
      sphereCollisionRadius=0
      boundBone {boneName=dyn_suit_bk_lt}
      boxExtents=<10 20 20>
      constraintSetup {angularLimitsMax=<1 -10 5>, angularLimitsMin=<-1 -45 -5>, angularTarget=<0 0 0>}
      localJointOffset=<2 0 0>
     }]
    relativeSpaceBone {boneName=root}
   }
   n30 = RigidBody(ComponentPose=@n29) {
    alphaInputType=Bool
    bClampLinearTranslationLimitToRefPose=true
    bOverrideWorldGravity=true
    overridePhysicsAsset=&M_MED_Alien_Summer_Physics
    simulationSpace=BaseBoneSpace
    alphaBoolBlend {blendInTime=0.1, blendOption=Linear, blendOutTime=0.1}
    baseBoneRef {boneName=pelvis}
    componentAppliedLinearAccClamp=<3000 3000 3000>
    componentLinearAccScale=<0.4 0.4 0.5>
    componentLinearVelScale=<0.25 0.25 0.35>
    overrideWorldGravity {z=-980}
   }
   n31 = Component To Local(ComponentPose=@n30)
   n32 = Get Shirt Active() -> (Shirt Active:bool)
   n33 = RigidBody!(
    ComponentPose=Local To Component(LocalPose=Blend Poses by bool(BlendPose_0=@n20, BlendPose_1=@n31, bActiveValue=@n32) {blendType=Linear})
    Alpha:float
    ExternalForce
   ) -> (Pose) {
    bClampLinearTranslationLimitToRefPose=true
    bOverrideWorldGravity=true
    overridePhysicsAsset=&/M_MED_Alien_Summer_Trunks_Physics
    simulationSpace=BaseBoneSpace
    alphaBoolBlend {blendOption=Linear}
    baseBoneRef {boneName=pelvis}
    componentAppliedLinearAccClamp=<3000 3000 3000>
    componentLinearAccScale=<0.3 0.3 0.3>
    componentLinearVelScale=<0.3 0.3 0.3>
    overrideWorldGravity {z=-980}
   }
   Output Pose(Result=Component To Local(ComponentPose=@n33)) {layerGroup=None}
   n34 = Control Rig!(Source=Copy Pose From Mesh() {bUseAttachedParent=true}, bAlphaBoolEnabled:bool=True) -> (Pose) {
    alphaInputType=Bool
    alphaBoolBlend {blendOption=Linear}
    controlRigAssetReference {blueprintRigClass=&/M_MED_Alien_Summer_Deform_CtrlRig_twists.M_MED_Alien_Summer_Deform_CtrlRig_twists_C}
   }
   n35 = Control Rig!(
    Source=@n34
    Alpha:float
    KneeRot_fr_l_114deg:Vector
    KneeTrans_fr_l_114deg:Vector="0.258,-3.591,0.009"
    KneeScale_fr_l_114deg:Vector="1,1,1"
    KneeScale_fr_l_130deg:Vector="1,1,1"
    KneeRot_fr_l_130deg:Vector="0,0,-51.2"
    KneeTrans_fr_l_130deg:Vector="0.258,-3.591,0.009"
   ) -> (Pose) {
    alphaBoolBlend {blendOption=Linear}
    controlRigAssetReference {blueprintRigClass=&/M_MED_Alien_Summer_Deform_CtrlRig_sculpters.M_MED_Alien_Summer_Deform_CtrlRig_sculpters_C}
   }
   n36 = Control Rig!(Source=@n35, Alpha=Get isCustomEmotePlaying() -> (isCustomEmotePlaying:double)) -> (Pose) {
    alphaBoolBlend {blendOption=Linear}
    alphaScaleBiasClamp {bInterpResult=true, interpSpeedDecreasing=20, interpSpeedIncreasing=20}
    controlRigAssetReference {blueprintRigClass=&/M_MED_Alien_Summer_CtrlRig_emote.M_MED_Alien_Summer_CtrlRig_emote_C}
   }
   n37 = "Transform (Modify) Bone - Bone: dyn_pelvis"(ComponentPose=Local To Component(LocalPose=@n36), Alpha=1, Translation="0,0,0", Rotation="0,0,0", Scale="1,0.5,0.5") {scaleMode=BMM_Additive, alphaBoolBlend {blendOption=Linear}, boneToModify {boneName=dyn_pelvis}, scale {y=0.5, z=0.5}}
   n38 = NOT Boolean(A=Get bCanPlayCustomAnimations() -> (bCanPlayCustomAnimations:bool))
   n39 = Get slope_foot_rot() -> (slope_foot_rot:Rotator)
   n40 = "Transform (Modify) Bone - Bone: dyn_slope_foot_l"!(
    ComponentPose=@n37
    bAlphaBoolEnabled:bool=@n38
    Translation=Get slope_foot_l_trans() -> (slope_foot_l_trans:Vector)
    Rotation=@n39
    Scale="1,1,1"
   ) -> (Pose) {
    alphaInputType=Bool
    rotationMode=BMM_Replace
    rotationSpace=BCS_WorldSpace
    translationMode=BMM_Replace
    translationSpace=BCS_WorldSpace
    alphaBoolBlend {blendOption=Linear}
    boneToModify {boneName=dyn_slope_foot_l}
   }
   n41 = "Transform (Modify) Bone - Bone: dyn_slope_foot_r"!(
    ComponentPose=@n40
    bAlphaBoolEnabled:bool=@n38
    Translation=Get slope_foot_r_trans() -> (slope_foot_r_trans:Vector)
    Rotation=@n39
    Scale="1,1,1"
   ) -> (Pose) {
    alphaInputType=Bool
    rotationMode=BMM_Replace
    rotationSpace=BCS_WorldSpace
    translationMode=BMM_Replace
    translationSpace=BCS_WorldSpace
    alphaBoolBlend {blendOption=Linear}
    boneToModify {boneName=dyn_slope_foot_r}
   }
   n42 = "Constraint - dyn_shirt_01_r_MSTR (R)"(ComponentPose=@n41, Alpha=1, ConstraintWeights_0=1) {
    alphaBoolBlend {blendOption=Linear}
    boneToModify {boneName=dyn_shirt_01_r_MSTR}
    constraintSetup [{transformType=Rotation, targetBone {boneName=thigh_r}}]
    constraintWeights [1]
   }
   n43 = "Constraint - dyn_shirt_fr_01_r_MSTR (RR)"(ComponentPose=@n42, Alpha=1, ConstraintWeights_0=0.75, ConstraintWeights_1=0.25) {
    alphaBoolBlend {blendOption=Linear}
    boneToModify {boneName=dyn_shirt_fr_01_r_MSTR}
    constraintSetup [{transformType=Rotation, targetBone {boneName=thigh_twist_01_r}}, {transformType=Rotation, targetBone {boneName=thigh_r}}]
    constraintWeights [1,1]
   }
   n44 = "Constraint - dyn_shirt_01_l_MSTR (R)"(ComponentPose=@n43, Alpha=1, ConstraintWeights_0=1) {
    alphaBoolBlend {blendOption=Linear}
    boneToModify {boneName=dyn_shirt_01_l_MSTR}
    constraintSetup [{transformType=Rotation, targetBone {boneName=thigh_l}}]
    constraintWeights [1]
   }
   n45 = "Constraint - dyn_shirt_fr_01_l_MSTR (RR)"(ComponentPose=@n44, Alpha=1, ConstraintWeights_0=0.75, ConstraintWeights_1=0.25) {
    alphaBoolBlend {blendOption=Linear}
    boneToModify {boneName=dyn_shirt_fr_01_l_MSTR}
    constraintSetup [{transformType=Rotation, targetBone {boneName=thigh_twist_01_l}}, {transformType=Rotation, targetBone {boneName=thigh_l}}]
    constraintWeights [1,1]
   }
   n46 = Get torsoscale() -> (torsoscale:bool)
   n47 = "Transform (Modify) Bone - Bone: dyn_spine_01"!(ComponentPose=@n45, bAlphaBoolEnabled:bool=@n46, Translation="0,0,0", Rotation="0,0,0", Scale="1,0.5,0.5") -> (Pose) {alphaInputType=Bool, scaleMode=BMM_Additive, alphaBoolBlend {blendOption=Linear}, boneToModify {boneName=dyn_spine_01}, scale {y=0.5, z=0.5}}
   n48 = "Transform (Modify) Bone - Bone: dyn_spine_02"!(ComponentPose=@n47, bAlphaBoolEnabled:bool=@n46, Translation="0,0,0", Rotation="0,0,0", Scale="1,0.5,0.5") -> (Pose) {alphaInputType=Bool, scaleMode=BMM_Additive, alphaBoolBlend {blendOption=Linear}, boneToModify {boneName=dyn_spine_02}, scale {y=0.5, z=0.5}}
   n49 = "Transform (Modify) Bone - Bone: dyn_spine_03"!(ComponentPose=@n48, bAlphaBoolEnabled:bool=@n46, Translation="0,0,0", Rotation="0,0,0", Scale="1,0.5,0.5") -> (Pose) {alphaInputType=Bool, scaleMode=BMM_Additive, alphaBoolBlend {blendOption=Linear}, boneToModify {boneName=dyn_spine_03}, scale {y=0.5, z=0.5}}
   n50 = "Transform (Modify) Bone - Bone: dyn_spine_04"!(ComponentPose=@n49, bAlphaBoolEnabled:bool=@n46, Translation="0,0,0", Rotation="0,0,0", Scale="1,0.5,0.5") -> (Pose) {alphaInputType=Bool, scaleMode=BMM_Additive, alphaBoolBlend {blendOption=Linear}, boneToModify {boneName=dyn_spine_04}, scale {y=0.5, z=0.5}}
   n51 = "Transform (Modify) Bone - Bone: dyn_spine_05"!(ComponentPose=@n50, bAlphaBoolEnabled:bool=@n46, Translation="0,0,0", Rotation="0,0,0", Scale="0.7,0.7,0.9") -> (Pose) {alphaInputType=Bool, scaleMode=BMM_Additive, alphaBoolBlend {blendOption=Linear}, boneToModify {boneName=dyn_spine_05}, scale=<0.7 0.7 0.9>}
   Save cached pose 'presim'(Pose=Component To Local(ComponentPose=@n51)) {cacheName=presim}
   NOT Boolean(A=OR Boolean(A=Get bIsTargeting(), B=Get bIsCrouching()))
   NOT Boolean(A=Get bIsCrouching())
   NOT Boolean(A=@n10)
   OR Boolean(A=Get bIsTargeting(), B=false)
   Get bIsJumping()
   Get thigh_rot_l()
   Get thigh_rot_r()
  ]}
 EventGraph {nodes [
   Event OnCharacterCustomizationLoaded()
   > n1 = Sequence()
   > n2 = Branch(Condition=CharacterItemDefinitionHasAnyMetaTag(Tags=Make Array!(
      "[0]"=(TagName="Cosmetics.Variant.Property.Stage1")
      "[1]":Gameplay Tag Structure=(TagName="Cosmetics.Variant.Property.Stage2")
      "[2]":Gameplay Tag Structure=(TagName="Cosmetics.Variant.Property.Stage3")
      "[3]":Gameplay Tag Structure=(TagName="Cosmetics.Variant.Property.Stage4")
      "[4]":Gameplay Tag Structure=(TagName="Cosmetics.Variant.Property.Stage5")
      "[5]":Gameplay Tag Structure=(TagName="Cosmetics.Variant.Property.Stage6")
      "[6]":Gameplay Tag Structure=(TagName="Cosmetics.Variant.Property.Stage7")
      "[7]":Gameplay Tag Structure=(TagName="Cosmetics.Variant.Property.Stage8")
     ) -> (Array)))
   > Set Shirt Active(Shirt Active:bool=true) -> (Output_Get:bool)
   Set Shirt Active(execute=@n2.else, Shirt Active=false)
   n3 = Make Array!(
    "[0]"=(TagName="Cosmetics.Variant.Property.Stage1")
    "[1]"=(TagName="Cosmetics.Variant.Property.Stage2")
    "[2]"=(TagName="Cosmetics.Variant.Property.Stage3")
    "[3]"=(TagName="Cosmetics.Variant.Property.Stage4")
    "[4]"=(TagName="Cosmetics.Variant.Property.Stage9")
   ) -> (Array)
   n4 = Get bRigidBodyEnabled()
   n5 = Branch(execute=@n1.then_1, Condition=AND Boolean(A=CharacterItemDefinitionHasAnyMetaTag(Tags=@n3), B=NOT Boolean(A=@n4)))
   > Set torsoscale(torsoscale:bool=true) -> (Output_Get:bool)
   Set torsoscale(execute=@n5.else, torsoscale=false)
   Event BlueprintUpdateAnimation()
   > Get!(execute:Exec) -> (then:Exec, else:Exec, Skel_Mesh:Skeletal Mesh Component Object Reference)
   > n6 = Sequence!(execute) -> (then_0, then_1, then_2:Exec, then_3:Exec, then_4:Exec)
   n7 = Get Skel_Mesh()
   > Set thigh_rot_l(thigh_rot_l=GetAlphaFromReferenceTransformDelta(SkeletalMeshComponent=@n7, BoneName=thigh_l, BaseName=pelvis, InputRange=(X=-20,Y=-100), TwistAxis="1,0,0"))
   > Set thigh_rot_r(thigh_rot_r=GetAlphaFromReferenceTransformDelta(SkeletalMeshComponent=@n7, BoneName=thigh_r, BaseName=pelvis, InputRange=(X=-20,Y=-100), TwistAxis="1,0,0"))
   n8 = Gravity Override State Function DEPRECATED(
    execute=@n6.then_1
    CharacterPartAnimInstance=Self-Reference()
    GameTimeInSeconds=GetGameTimeInSeconds()
    GravityOverrideParams=Get Gravity Override Params()
   )
   > Set GravityOverride(GravityOverride=@n8)
   n9 = Break Fort Anim Input Slope Info!(FortAnimInput_SlopeInfo:"Fort Anim Input Slope Info Structure (by ref)"=Get SlopeInfoInput() -> (SlopeInfoInput:Fort Anim Input Slope Info Structure)) -> (RootSlopeTranslation:Vector=~, LeftFootSlopeTranslation:Vector=~, RightFootSlopeTranslation:Vector=~, SlopeRotation:Rotator=~) {is=title K2Node_BreakStruct}
   Set slope_foot_l_trans(execute=@n6.then_2, slope_foot_l_trans:Vector=@n9.LeftFootSlopeTranslation) -> (Output_Get:Vector)
   > Set slope_foot_r_trans(slope_foot_r_trans:Vector=@n9.RightFootSlopeTranslation) -> (Output_Get:Vector)
   > Set slope_foot_rot(slope_foot_rot:Rotator=@n9.SlopeRotation) -> (Output_Get:Rotator)
   n10 = GetAnimInstance(self=GetOwnerSkeletalMeshComponent())
   Is Valid(exec=@n6.then_3, InputObject=@n10)
   n11 = Montage_IsPlaying(self=@n10, Montage=/BRCosmetics/Animation/Game/MainPlayer/Emotes/Hip_Hop_Gesture_1/Emote_Hip_Hop_Gesture1_CMM_M)
   n12 = Montage_IsPlaying(self=@n10, Montage=/BRCosmetics/Animation/Game/MainPlayer/Emotes/RockPaperScissor/Emote_RockPaperScissor_Paper_CMM_M)
   n13 = Montage_IsPlaying(self=@n10, Montage=/BRCosmetics/Animation/Game/MainPlayer/Emotes/IndigoApple/Emote_IndigoApple_CMM_M)
   > Set isCustomEmotePlaying(isCustomEmotePlaying:double=SelectFloat(A=1, B=0, bPickA=OR Boolean!(A=@n11, B=@n12, C:bool=@n13, D:bool, E:bool) -> (ReturnValue))) -> (Output_Get:double)
   Event BlueprintInitializeAnimation()
   n14 = GetOwnerSkeletalMeshComponent()
   > Is Valid(InputObject=@n14)
   > Set Skel_Mesh(Skel_Mesh=@n14)
   > Set GetPlayerListUsesCustomCharacterParts(GetPlayerListUsesCustomCharacterParts:bool=GetPlaylistUsesCustomCharacterParts!() -> (ReturnValue:bool)) -> (Output_Get:bool)
   n15 = Branch(Condition=OR Boolean!(A=Get bIsCrouching(), B=Get bIsJumping(), C=Get bIsSprinting()) -> (ReturnValue))
   Set thigh_rot_l(execute=@n15.else, thigh_rot_l=0)
   > Set thigh_rot_r(thigh_rot_r=0)
   n16 = Branch(Condition=Get bIsCrouching())
   n17 = BreakTransform(InTransform=GetDeltaTransformFromRefPose(self=GetOwnerSkeletalMeshComponent(), BoneName=thigh_twist_01_l, BaseName=pelvis))
   > Set leg rotation_l(leg rotation_l:double=MapRangeClamped(Value=BreakRotator(InRot=@n17.Rotation), InRangeA=10, InRangeB=4)) -> (Output_Get:double)
   Set leg rotation_l(execute=@n16.else, leg rotation_l=0)
  ]}
]
variables [
 GetPlayerListUsesCustomCharacterParts {type=boolean}
 Gravity Override Params {type=GravityOverrideParamsStruct}
 GravityOverride {type=Vector}
 isCustomEmotePlaying {type=number}
 leg rotation_l {type=number}
 Shirt Active {type=boolean}
 Skel_Mesh {type=SkeletalMeshComponent}
 slope_foot_l_trans {type=Vector}
 slope_foot_r_trans {type=Vector}
 slope_foot_rot {type=Rotator}
 thigh_rot_l {type=number}
 thigh_rot_r {type=number}
 torsoscale {type=boolean}
]
