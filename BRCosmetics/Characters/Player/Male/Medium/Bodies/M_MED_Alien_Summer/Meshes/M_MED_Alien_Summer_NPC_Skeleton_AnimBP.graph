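# Anim Blueprint for the M_MED_Alien_Summer NPC skeleton that applies an emote-specific Control Rig to the Alien Summer cosmetic body.
# Copy Pose From Mesh takes the attached parent's pose and feeds the Control Rig, whose alpha follows isCustomEmotePlaying with interpolated blending.
# Each update sets that float to 1 or 0 by checking the owning component's anim instance for Hip Hop Gesture, RockPaperScissor Paper or IndigoApple montages.
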AnimBlueprint /BRCosmetics/Characters/Player/Male/Medium/Bodies/M_MED_Alien_Summer/Meshes/M_MED_Alien_Summer_NPC_Skeleton_AnimBP
target_skeleton=&/M_MED_Alien_Summer_NPC_Skeleton
graphs [
 AnimGraph {nodes [
   n1 = Control Rig(Source=Copy Pose From Mesh() {bUseAttachedParent=true}, Alpha=Get isCustomEmotePlaying() -> (isCustomEmotePlaying:double)) {
    alphaBoolBlend {blendOption=Linear}
    alphaScaleBiasClamp {bInterpResult=true, interpSpeedDecreasing=20, interpSpeedIncreasing=20}
    controlRigAssetReference {blueprintRigClass=&/M_MED_Alien_Summer_CtrlRig_emote.M_MED_Alien_Summer_CtrlRig_emote_C}
   }
   Output Pose(Result=@n1) {layerGroup=None}
  ]}
 EventGraph {nodes [
   Event BlueprintUpdateAnimation()
   n1 = GetAnimInstance(self=GetOwnerSkeletalMeshComponent())
   > Is Valid(InputObject=@n1)
   n2 = Montage_IsPlaying(self=@n1, Montage=/BRCosmetics/Animation/Game/MainPlayer/Emotes/Hip_Hop_Gesture_1/Emote_Hip_Hop_Gesture1_CMM_M)
   n3 = Montage_IsPlaying(self=@n1, Montage=/BRCosmetics/Animation/Game/MainPlayer/Emotes/RockPaperScissor/Emote_RockPaperScissor_Paper_CMM_M)
   n4 = Montage_IsPlaying(self=@n1, Montage=/BRCosmetics/Animation/Game/MainPlayer/Emotes/IndigoApple/Emote_IndigoApple_CMM_M)
   > Set isCustomEmotePlaying(isCustomEmotePlaying:double=SelectFloat(A=1, B=0, bPickA=OR Boolean!(A=@n2, B=@n3, C:bool=@n4, D:bool, E:bool) -> (ReturnValue))) -> (Output_Get:double)
   Event BlueprintInitializeAnimation()
   > Set Skel_Mesh(Skel_Mesh=GetOwnerSkeletalMeshComponent())
   TryGetPawnOwner()
  ]}
]
variables [
 isCustomEmotePlaying {type=number}
 Skel_Mesh {type=SkeletalMeshComponent}
]
