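# Physics asset for the M_MED_Alien_Summer male player cosmetic body, covering only its dynamic headphone bones.
# A large capsule on dyn_Headphones and a tiny one on dyn_Headphones_MSTR form the two bodies.
# One constraint joins them with tight swing and twist limits, allowing only slight angular movement of the headphones relative to the master bone.
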PhysicsAsset /BRCosmetics/Characters/Player/Male/Medium/Bodies/M_MED_Alien_Summer/Meshes/Parts/M_MED_Alien_Summer_Physics
bodies {
 dyn_Headphones [{shapeType=Capsule, length=14.5656, radius=6.3117, center=<2 0 0>, rotation=<rot 0 0 90>}]
 dyn_Headphones_MSTR [{shapeType=Capsule, length=0.505, radius=0.505, rotation=<rot 0 -90 0>}]
}
constraints [{bone1Name=dyn_Headphones, bone2Name=dyn_Headphones_MSTR, swing1LimitDegrees=1, swing2LimitDegrees=7, twistLimitDegrees=1}]
