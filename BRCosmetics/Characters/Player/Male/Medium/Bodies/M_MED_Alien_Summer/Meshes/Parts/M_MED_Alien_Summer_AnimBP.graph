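# Animation Blueprint for the M_MED_Alien_Summer character body part, which follows the pose of its attached parent mesh.
# Copy Pose From Mesh feeds an additive Transform Bone on dyn_retro_glasses with zero translation, then the output pose.
# A head-based RigidBody simulation and an angular-spring AnimDynamics node for dyn_Headphones exist but do not connect to the output.
# The event graph has no logic.
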AnimBlueprint /BRCosmetics/Characters/Player/Male/Medium/Bodies/M_MED_Alien_Summer/Meshes/Parts/M_MED_Alien_Summer_AnimBP
target_skeleton=&/M_MED_Alien_Summer_Skeleton
graphs [
 AnimGraph {nodes [
   n1 = RigidBody(ComponentPose=Local To Component()) {
    bClampLinearTranslationLimitToRefPose=true
    bOverrideWorldGravity=true
    overridePhysicsAsset=&/M_MED_Alien_Summer_Physics
    simulationSpace=BaseBoneSpace
    alphaBoolBlend {blendOption=Linear}
    baseBoneRef {boneName=head}
    componentAppliedLinearAccClamp=<3000 3000 3000>
    componentLinearAccScale=<0.3 0.3 0.3>
    componentLinearVelScale=<0.2 0.2 0.2>
    overrideWorldGravity {z=-980}
   }
   Component To Local(ComponentPose=@n1)
   n2 = Anim Dynamics - Bone: dyn_Headphones!(ComponentPose=Local To Component(), bAlphaBoolEnabled:bool=True, GravityOverride:Vector="0,0,0", ExternalForce="0,0,0") -> (Pose) {
    alphaInputType=Bool
    angularDampingOverride=0.999
    angularSpringConstant=300
    bAngularSpring=true
    bGravityOverrideInSimSpace=true
    bOverrideAngularDamping=true
    simulationSpace=RootRelative
    alphaBoolBlend {blendInTime=0.1, blendOption=Linear, blendOutTime=0.1}
    boundBone {boneName=dyn_Headphones}
    componentAppliedLinearAccClamp=<3500 3500 4300>
    componentLinearAccScale=<3 3 3>
    componentLinearVelScale=<3 3 3>
    physicsBodyDefinitions [{
      sphereCollisionRadius=0
      boundBone {boneName=dyn_Headphones}
      boxExtents=<10 30 15>
      constraintSetup {angularLimitsMax=<0 5 0>, angularLimitsMin=<0 -5 0>, angularTarget=<0 0 0>}
      localJointOffset=<-2 0 0>
     }]
    relativeSpaceBone {boneName=spine_05}
   }
   Component To Local(ComponentPose=@n2)
   n3 = Local To Component(LocalPose=Copy Pose From Mesh() {bUseAttachedParent=true})
   Output Pose(Result=Component To Local(ComponentPose="Transform (Modify) Bone - Bone: dyn_retro_glasses"(ComponentPose=@n3, Alpha=1, Translation="0,0,0", Rotation="0,0,0", Scale="1,1,1") {translationMode=BMM_Additive, alphaBoolBlend {blendOption=Linear}, boneToModify {boneName=dyn_retro_glasses}})) {layerGroup=None}
  ]}
 EventGraph {nodes [Event BlueprintUpdateAnimation()]}
]
variables []
